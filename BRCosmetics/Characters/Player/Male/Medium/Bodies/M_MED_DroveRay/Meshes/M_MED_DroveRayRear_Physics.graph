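# Physics asset for the rear mesh of the DroveRay medium male player body, covering its scarf and upper body.
# Capsules on pelvis, spine, clavicles and upper arms form body volumes, linked mostly by free angular constraints.
# A four-bone back scarf chain hangs off spine_05 through a base bone, with moderate swing and very tight twist limits.
# The single right scarf bone has zero-degree limits against its base.
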PhysicsAsset /BRCosmetics/Characters/Player/Male/Medium/Bodies/M_MED_DroveRay/Meshes/M_MED_DroveRayRear_Physics
bodies {
 base_scarf_back [{shapeType=Capsule, length=7, radius=2, shapeName=base_scarf_back_capsule, center=<0 0 1.6>, rotation=<rot -0.0001 0 90.7133>}]
 base_scarf_r [{shapeType=Capsule, length=0.505, radius=0.505, shapeName=base_scarf_r_capsule, rotation=<rot 0 -90 0>}]
 clavicle_l [{shapeType=Capsule, length=10, radius=7.5, shapeName=clavicle_l_capsule, center=<4 0 0>, rotation=<rot 90 0 0>}]
 clavicle_r [{shapeType=Capsule, length=10, radius=7.5, shapeName=clavicle_r_capsule, center=<-4 0.6507 0>, rotation=<rot 90 178.8404 0>}]
 dyn_scarf_back_01 [
  {shapeType=Capsule, length=10, radius=2.25, shapeName=dyn_scarf_back_01_capsule, center=<1.5 0 1.4>, rotation=<rot -0.0795 -0.0143 98.1871>}
  {
   shapeType=Capsule
   length=10
   radius=2.25
   shapeName=dyn_scarf_back_01_capsule
   center=<6 -0.0096 1.1906>
   rotation=<rot -0.0007 -0.0001 92.6121>
  }
 ]
 dyn_scarf_back_02 [
  {shapeType=Capsule, length=10, radius=2, shapeName=dyn_scarf_back_02_capsule, center=<1 -0.0301 0.7402>, rotation=<rot -0.0003 0 89.6241>}
  {shapeType=Capsule, length=10, radius=2, shapeName=dyn_scarf_back_02_capsule, center=<5.7616 0.012 0.9807>, rotation=<rot -0.0001 0 88.6719>}
 ]
 dyn_scarf_back_03 [{shapeType=Capsule, length=10, radius=2.6, shapeName=dyn_scarf_back_03_capsule, center=<2 0 0.612>, rotation=<rot 0 0 90>}]
 dyn_scarf_back_04 [{shapeType=Capsule, length=10, radius=3, shapeName=dyn_scarf_back_04_capsule, center=<4 -0.1071 0.5>, rotation=<rot -0.0001 0 83.4154>}]
 dyn_scarf_r [{shapeType=Capsule, length=4.5, radius=1.75, shapeName=dyn_scarf_r_capsule, rotation=<rot 76.4485 89.9999 -180>}]
 pelvis [{
   shapeType=Capsule
   length=7.3331
   radius=11.5673
   shapeName=pelvis_capsule
   center=<-1.1476 -1.0896 0.0585>
   rotation=<rot 0.3144 -90 0.1309>
  }]
 spine_01 [{
   shapeType=Capsule
   length=7.8368
   radius=11.7134
   shapeName=spine_01_capsule
   center=<1.9464 0.147 -0.1553>
   rotation=<rot 1.9672 -90 -0.3479>
  }]
 spine_02 [{shapeType=Capsule, length=8, radius=10.25, center=<3.3441 -0.6224 0>, rotation=<rot 0 5.0673 0>}]
 spine_03 [{shapeType=Capsule, length=8, radius=11, shapeName=spine_03_capsule, center=<6.2426 0.3073 0>}]
 spine_04 [{shapeType=Capsule, length=8, radius=11.5, center=<5.4722 0.4482 0>}]
 spine_05 [{shapeType=Capsule, length=10, radius=11.5, center=<0.6421 0.4955 0>}]
 upperarm_l [{shapeType=Capsule, length=18, radius=6, shapeName=upperarm_l_capsule, center=<6.6728 -0.9342 -1.1423>, rotation=<rot 88.676 -0.0057 -4.7847>}]
 upperarm_r [{
   shapeType=Capsule
   length=26
   radius=5.5
   shapeName=upperarm_r_capsule
   center=<-7.0299 1.2917 1.143>
   rotation=<rot 88.676 -0.0057 175.2153>
  }]
}
constraints [
 {bone1Name=spine_04, bone2Name=spine_03, swing1LimitDegrees=45, swing2LimitDegrees=45, twistLimitDegrees=45}
 {bone1Name=spine_05, bone2Name=spine_04, swing1LimitDegrees=45, swing2LimitDegrees=45, twistLimitDegrees=45}
 {
  bone1Name=spine_03
  bone2Name=spine_02
  swing1LimitDegrees=45
  swing1Motion=Free
  swing2LimitDegrees=45
  swing2Motion=Free
  twistLimitDegrees=45
  twistMotion=Free
 }
 {
  bone1Name=base_scarf_back
  bone2Name=spine_05
  swing1LimitDegrees=45
  swing1Motion=Free
  swing2LimitDegrees=45
  swing2Motion=Free
  twistLimitDegrees=45
  twistMotion=Free
 }
 {bone1Name=dyn_scarf_back_01, bone2Name=base_scarf_back, swing1LimitDegrees=30, swing2LimitDegrees=40, twistLimitDegrees=3.5}
 {bone1Name=dyn_scarf_back_02, bone2Name=dyn_scarf_back_01, swing1LimitDegrees=28, swing2LimitDegrees=45, twistLimitDegrees=3}
 {bone1Name=dyn_scarf_back_03, bone2Name=dyn_scarf_back_02, swing1LimitDegrees=30, swing2LimitDegrees=45, twistLimitDegrees=3}
 {bone1Name=dyn_scarf_back_04, bone2Name=dyn_scarf_back_03, swing1LimitDegrees=30, swing2LimitDegrees=45, twistLimitDegrees=2}
 {
  bone1Name=clavicle_l
  bone2Name=spine_05
  swing1LimitDegrees=45
  swing1Motion=Free
  swing2LimitDegrees=45
  swing2Motion=Free
  twistLimitDegrees=45
  twistMotion=Free
 }
 {
  bone1Name=upperarm_l
  bone2Name=clavicle_l
  swing1LimitDegrees=45
  swing1Motion=Free
  swing2LimitDegrees=45
  swing2Motion=Free
  twistLimitDegrees=45
  twistMotion=Free
 }
 {
  bone1Name=clavicle_r
  bone2Name=spine_05
  swing1LimitDegrees=45
  swing1Motion=Free
  swing2LimitDegrees=45
  swing2Motion=Free
  twistLimitDegrees=45
  twistMotion=Free
 }
 {
  bone1Name=upperarm_r
  bone2Name=clavicle_r
  swing1LimitDegrees=45
  swing1Motion=Free
  swing2LimitDegrees=45
  swing2Motion=Free
  twistLimitDegrees=45
  twistMotion=Free
 }
 {
  bone1Name=base_scarf_r
  bone2Name=spine_05
  swing1LimitDegrees=45
  swing1Motion=Free
  swing2LimitDegrees=45
  swing2Motion=Free
  twistLimitDegrees=45
  twistMotion=Free
 }
 {bone1Name=dyn_scarf_r, bone2Name=base_scarf_r, swing1LimitDegrees=0, swing2LimitDegrees=0, twistLimitDegrees=0}
 {
  bone1Name=spine_02
  bone2Name=spine_01
  swing1LimitDegrees=45
  swing1Motion=Free
  swing2LimitDegrees=45
  swing2Motion=Free
  twistLimitDegrees=45
  twistMotion=Free
 }
 {
  bone1Name=spine_01
  bone2Name=pelvis
  swing1LimitDegrees=45
  swing1Motion=Free
  swing2LimitDegrees=45
  swing2Motion=Free
  twistLimitDegrees=45
  twistMotion=Free
 }
]
