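# Control Rig for the DroveRay male character body that corrects scarf, knee, groin and foot helper bones.
# After a shared deform rig, weighted parent constraints attach helper bones to the body, and variables for targeting and equipped weapons are smoothed.
# Spherical pose readers and a swing/twist DeltaRotation function then weight additive bone offsets, many gated by states like parachuting or DBNO.
# Scarf tips pin to spine_03 when rigid bodies are disabled.
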ControlRigBlueprint /BRCosmetics/Characters/Player/Male/Medium/Bodies/M_MED_DroveRay/Meshes/M_MED_DroveRay_CtrlRig
graphs [
 New Function_ContainedGraph {nodes [
   n1 = RigVMFunctionEntryNode()
   RigVMFunction_MathBoolToInteger = &RigVMFunction_MathBoolToInteger(Value=@n1.onlyTwist)
   RigVMFunction_MathQuaternionSwingTwist = &RigVMFunction_MathQuaternionSwingTwist(
    twistAxis=@n1.twistAxis
    Input=&RigUnit_GetRelativeTransformForItem(bChildInitial=False, bParentInitial=False, child=@n1.child, parent=@n1.parent).RelativeTransform.Rotation
   )
   n2 = RigVMDispatchNode(
    Values=((),())
    Index=@RigVMFunction_MathBoolToInteger.Result
    Values.0=@RigVMFunction_MathQuaternionSwingTwist.Swing
    Values.1=@RigVMFunction_MathQuaternionSwingTwist.Twist
   ) {template="DISPATCH_RigVMDispatch_SelectInt32(in Index,in Values,out Result)"}
   RigVMFunction_MathQuaternionToEuler = &RigVMFunction_MathQuaternionToEuler(Value=@n2.Result, RotationOrder=@n1.Rotation Order)
   > DISPATCH_RigVMDispatch_Print = RigVMDispatchNode(ScreenColor=(R=1,G=1,B=1,A=1), ScreenDuration=1.5, Enabled=@n1.debugRotation, Value=@RigVMFunction_MathQuaternionToEuler.Result) {template="DISPATCH_RigVMDispatch_Print(in Prefix,in Value,in Enabled,in ScreenDuration,in ScreenColor)"}
   Cast = RigVMDispatchNode(Value=@n1.triggerAxis) {template="DISPATCH_RigVMDispatch_CastEnumToInt(in Value,out Result)"}
   DISPATCH_RigVMDispatch_SelectInt32 = RigVMDispatchNode(
    Values=(0,0,0)
    Index=@Cast.Result
    Values.0=@RigVMFunction_MathQuaternionToEuler.Result.X
    Values.1=@RigVMFunction_MathQuaternionToEuler.Result.Y
    Values.2=@RigVMFunction_MathQuaternionToEuler.Result.Z
   ) {template="DISPATCH_RigVMDispatch_SelectInt32(in Index,in Values,out Result)"}
   Remap = &RigVMFunction_MathDoubleRemap(
    bClamp=False
    Value=@DISPATCH_RigVMDispatch_SelectInt32.Result
    TargetMinimum=@n1.outputMinMax.X
    TargetMaximum=@n1.outputMinMax.Y
    SourceMinimum=@n1.rotationMinMax.X
    SourceMaximum=@n1.rotationMinMax.Y
   )
   Absolute = &RigVMFunction_MathDoubleAbs(Value=@Remap.Result)
   Multiply = &RigVMFunction_MathDoubleMul(
    A=&RigVMFunction_MathDoubleClamp(Value=RigVMDispatchNode(True=@Absolute.Result, Condition=@n1.absoluteValue, False=@Remap.Result) {template="DISPATCH_RigVMDispatch_If(in Condition,in True,in False,out Result)"}.Result).Result
    B=@n1.premult
   )
   > RigVMFunctionReturnNode(outputQuat=@n2.Result, outputWeight=@Multiply.Result)
  ]}
 New Function_ContainedGraph {nodes [
   RigVMFunctionEntryNode()
   > &RigUnit_ParentConstraint(
    AdvancedSettings=(RotationOrderForFilter=XZY)
    Child=(Type=Bone,Name="groin_l_helper")
    Parents="((Item=(Type=Bone,Name=\"ball_l\"),Weight=0.0005),(Item=(Type=Bone,Name=\"deform_groin_r\"),Weight=0.0007),(Item=(Type=Bone,Name=\"thigh_twist_01_r\"),Weight=0.0185),(Item=(Type=Bone,Name=\"thigh_l\"),Weight=0.0189),(Item=(Type=Bone,Name=\"spine_01\"),Weight=0.0339) ...(truncated)"
    Weight=1
   )
   > &RigUnit_ParentConstraint(
    AdvancedSettings=(RotationOrderForFilter=XZY)
    Child=(Type=Bone,Name="base_scarf_front")
    Parents="((Item=(Type=Bone,Name=\"upperarm_twist_01_r\"),Weight=0.0008),(Item=(Type=Bone,Name=\"neck_01\"),Weight=0.0032),(Item=(Type=Bone,Name=\"neck_02\"),Weight=0.0236),(Item=(Type=Bone,Name=\"spine_03\"),Weight=0.0777),(Item=(Type=Bone,Name=\"clavicle_r\"),Weight=0.1677) ...(truncated)"
    Weight=1
   )
   > &RigUnit_ParentConstraint(
    AdvancedSettings=(RotationOrderForFilter=XZY)
    Child=(Type=Bone,Name="base_scarf_r")
    Parents="((Item=(Type=Bone,Name=\"head\"),Weight=0.0004),(Item=(Type=Bone,Name=\"upperarm_twist_01_r\"),Weight=0.001),(Item=(Type=Bone,Name=\"spine_04\"),Weight=0.0069),(Item=(Type=Bone,Name=\"spine_03\"),Weight=0.0362),(Item=(Type=Bone,Name=\"neck_01\"),Weight=0.0488),(Item ...(truncated)"
    Weight=1
   )
   > &RigUnit_ParentConstraint(
    AdvancedSettings=(RotationOrderForFilter=XZY)
    Child=(Type=Bone,Name="base_scarf_back")
    Parents="((Item=(Type=Bone,Name=\"spine_04\"),Weight=0.0151),(Item=(Type=Bone,Name=\"neck_01\"),Weight=0.0215),(Item=(Type=Bone,Name=\"neck_02\"),Weight=0.0736),(Item=(Type=Bone,Name=\"spine_03\"),Weight=0.0827),(Item=(Type=Bone,Name=\"clavicle_r\"),Weight=0.3022),(Item=(Typ ...(truncated)"
    Weight=1
   )
   > &RigUnit_ParentConstraint(
    AdvancedSettings=(RotationOrderForFilter=XZY)
    Child=(Type=Bone,Name="base_scarf_back_middle")
    Parents="((Item=(Type=Bone,Name=\"clavicle_r\"),Weight=0.007),(Item=(Type=Bone,Name=\"spine_03\"),Weight=0.0078),(Item=(Type=Bone,Name=\"clavicle_l\"),Weight=0.0116),(Item=(Type=Bone,Name=\"head\"),Weight=0.0616),(Item=(Type=Bone,Name=\"neck_01\"),Weight=0.1437),(Item=(Type= ...(truncated)"
    Weight=1
   )
   > &RigUnit_ParentConstraint(
    AdvancedSettings=(RotationOrderForFilter=XZY)
    Child=(Type=Bone,Name="base_scarf_front_middle")
    Parents="((Item=(Type=Bone,Name=\"clavicle_r\"),Weight=0.0097),(Item=(Type=Bone,Name=\"spine_03\"),Weight=0.0128),(Item=(Type=Bone,Name=\"head\"),Weight=0.022),(Item=(Type=Bone,Name=\"spine_04\"),Weight=0.0412),(Item=(Type=Bone,Name=\"neck_02\"),Weight=0.0685),(Item=(Type=Bo ...(truncated)"
    Weight=1
   )
   > &RigUnit_ParentConstraint(
    AdvancedSettings=(RotationOrderForFilter=XZY)
    Child=(Type=Bone,Name="base_inner_r_knee")
    Parents="((Item=(Type=Bone,Name=\"calf_l\"),Weight=0.0002),(Item=(Type=Bone,Name=\"thigh_l\"),Weight=0.0003),(Item=(Type=Bone,Name=\"calf_twist_02_r\"),Weight=0.0071),(Item=(Type=Bone,Name=\"deform_knee_out_r\"),Weight=0.0075),(Item=(Type=Bone,Name=\"thigh_twist_01_r\"),Weig ...(truncated)"
    Weight=1
   )
   > &RigUnit_ParentConstraint(
    AdvancedSettings=(RotationOrderForFilter=XZY)
    Child=(Type=Bone,Name="base_outer_r_knee")
    Parents="((Item=(Type=Bone,Name=\"deform_knee_in_r\"),Weight=0.0001),(Item=(Type=Bone,Name=\"pelvis\"),Weight=0.0002),(Item=(Type=Bone,Name=\"thigh_twist_01_r\"),Weight=0.0022),(Item=(Type=Bone,Name=\"deform_knee_out_r\"),Weight=0.0044),(Item=(Type=Bone,Name=\"calf_twist_02 ...(truncated)"
    Weight=1
   )
   > &RigUnit_ParentConstraint(
    AdvancedSettings=(RotationOrderForFilter=XZY)
    Child=(Type=Bone,Name="base_inner_l_knee")
    Parents="((Item=(Type=Bone,Name=\"calf_r\"),Weight=0.0002),(Item=(Type=Bone,Name=\"thigh_r\"),Weight=0.0003),(Item=(Type=Bone,Name=\"calf_twist_02_l\"),Weight=0.0068),(Item=(Type=Bone,Name=\"thigh_twist_01_l\"),Weight=0.009),(Item=(Type=Bone,Name=\"thigh_l\"),Weight=0.3649), ...(truncated)"
    Weight=1
   )
   > &RigUnit_ParentConstraint(
    AdvancedSettings=(RotationOrderForFilter=XZY)
    Child=(Type=Bone,Name="base_outer_l_knee")
    Parents="((Item=(Type=Bone,Name=\"deform_knee_in_l\"),Weight=0.0002),(Item=(Type=Bone,Name=\"deform_knee_out_l\"),Weight=0.0004),(Item=(Type=Bone,Name=\"thigh_twist_01_l\"),Weight=0.0019),(Item=(Type=Bone,Name=\"calf_twist_02_l\"),Weight=0.0269),(Item=(Type=Bone,Name=\"thig ...(truncated)"
    Weight=1
   )
   > RigVMFunctionReturnNode()
  ]}
 New Function_ContainedGraph {nodes [
   RigVMFunctionEntryNode()
   > RigVMFunction_Sequence = RigVMAggregateNode()
   RigVMFunctionReturnNode(ExecuteContext=@RigVMFunction_Sequence.H)
   n1 = &RigUnit_SphericalPoseReader(
    ActiveRegionScaleFactors=(PositiveWidth=1,NegativeWidth=1,PositiveHeight=1,NegativeHeight=1)
    ActiveRegionSize=0.06
    Debug=(bDrawDebug=false,bDraw2D=False,bDrawLocalAxes=False,DebugScale=25,DebugSegments=20,DebugThickness=0.25)
    DriverAxis=(X=1,Y=0,Z=0)
    DriverItem=(Type=Bone,Name="head")
    FalloffRegionScaleFactors=(PositiveWidth=1,NegativeWidth=1,PositiveHeight=1,NegativeHeight=1)
    FalloffSize=0.25
    FlipHeightScaling=False
    FlipWidthScaling=False
    OptionalParentItem=(Type=Bone,Name="spine_05")
    RotationOffset=(X=-10,Y=-60,Z=-50)
    ExecutePin=@RigVMFunction_Sequence.D
   )
   n2 = RigVMVariableNode() {variable=isTargeting_Lerp}
   AlphaInterp = &RigVMFunction_AlphaInterp(
    InterpSpeedDecreasing=3
    InterpSpeedIncreasing=3
    Scale=1
    bClampResult=False
    Value=&RigVMFunction_MathDoubleMul(
     A=&RigVMFunction_MathDoubleMul(A=@n1.OutputParam, B=@n2.Value).Result
     B=RigVMVariableNode() {variable=isAR_Equipped_Lerp}.Value
    ).Result
   )
   > &RigUnit_ModifyTransforms(
    ItemToModify=((Item=(Type=Bone,Name="base_scarf_r"),Transform=(Rotation=(X=0,Y=0,Z=0,W=1),Translation=(X=1,Y=3.5,Z=2.5),Scale3D=(X=1,Y=1,Z=1))))
    WeightMinimum=0
    Weight=@AlphaInterp.Result
   )
   > n3 = &RigUnit_SphericalPoseReader(
    ActiveRegionScaleFactors=(PositiveWidth=1,NegativeWidth=1,PositiveHeight=1,NegativeHeight=1)
    ActiveRegionSize=0.05
    Debug=(bDrawDebug=false,bDraw2D=False,bDrawLocalAxes=False,DebugScale=25,DebugSegments=20,DebugThickness=0.25)
    DriverAxis=(X=1,Y=0,Z=0)
    DriverItem=(Type=Bone,Name="head")
    FalloffRegionScaleFactors=(PositiveWidth=1,NegativeWidth=1,PositiveHeight=1,NegativeHeight=1)
    FalloffSize=0.35
    FlipHeightScaling=False
    FlipWidthScaling=False
    OptionalParentItem=(Type=Bone,Name="spine_05")
    RotationOffset=(X=0,Y=-65,Z=-52)
   )
   n4 = RigVMVariableNode() {variable=isTargeting_Lerp}
   > &RigUnit_ModifyTransforms(
    ItemToModify=((Item=(Type=Bone,Name="base_scarf_r"),Transform=(Rotation=(X=0,Y=0,Z=0,W=1),Translation=(X=1,Y=1.5,Z=1.5),Scale3D=(X=1,Y=1,Z=1))))
    WeightMinimum=0
    Weight=&RigVMFunction_MathDoubleMul(
     A=&RigVMFunction_MathDoubleMul(A=@n3.OutputParam, B=@n4.Value).Result
     B=RigVMVariableNode() {variable=isPistol_Equipped_Lerp}.Value
    ).Result
   )
   > &RigUnit_ModifyTransforms(
    ItemToModify=((Item=(Type=Bone,Name="base_scarf_front"),Transform=(Rotation=(X=0,Y=-0.766,Z=0,W=0.6428),Translation=(X=0,Y=0,Z=0),Scale3D=(X=1,Y=1,Z=1))))
    WeightMinimum=0
    Weight=RigVMDispatchNode(False=0, True=1, Condition=RigVMVariableNode() {variable=isDBNO}.Value) {template="DISPATCH_RigVMDispatch_If(in Condition,in True,in False,out Result)"}.Result
   )
   RigVMFunction_ControlFlowBranch = &RigVMFunction_ControlFlowBranch(Condition=RigVMVariableNode() {variable=pub_isRigidBodies_Enabled}.Value, ExecuteContext=@RigVMFunction_Sequence.B)
   &RigUnit_ParentConstraint(
    AdvancedSettings=(InterpolationType=Average,RotationOrderForFilter=XZY)
    Child=(Type=Bone,Name="dyn_scarf_front_03")
    Parents=((Item=(Type=Bone,Name="spine_03"),Weight=1))
    Weight=1
    ExecutePin=@RigVMFunction_ControlFlowBranch.False
   )
   > &RigUnit_ParentConstraint(
    AdvancedSettings=(InterpolationType=Average,RotationOrderForFilter=XZY)
    Child=(Type=Bone,Name="dyn_scarf_back_03")
    Parents=((Item=(Type=Bone,Name="spine_03"),Weight=1))
    Weight=1
   )
   n5 = &RigUnit_SphericalPoseReader(
    ActiveRegionScaleFactors=(PositiveWidth=1,NegativeWidth=1,PositiveHeight=1,NegativeHeight=0.8)
    ActiveRegionSize=0.1
    Debug=(bDrawDebug=false,bDraw2D=False,bDrawLocalAxes=False,DebugScale=25,DebugSegments=20,DebugThickness=0.25)
    DriverAxis=(X=1,Y=0,Z=0)
    DriverItem=(Type=Bone,Name="upperarm_r")
    FalloffRegionScaleFactors=(PositiveWidth=1,NegativeWidth=1,PositiveHeight=1.7,NegativeHeight=0.5)
    FalloffSize=0.2
    FlipHeightScaling=False
    FlipWidthScaling=False
    OptionalParentItem=(Type=Bone,Name="None")
    RotationOffset=(X=51.9395,Y=8.7023,Z=22.4256)
    ExecutePin=@RigVMFunction_Sequence.C
   )
   > &RigUnit_ModifyTransforms(
    ItemToModify="((Item=(Type=Bone,Name=\"base_scarf_r\"),Transform=(Rotation=(X=0.1305,Y=0,Z=0,W=0.9914),Translation=(X=2,Y=0,Z=2),Scale3D=(X=1,Y=1,Z=1))),(Item=(Type=Bone,Name=\"upperarm_r_dup\"),Transform=(Rotation=(X=-0.1305,Y=0,Z=0,W=0.9914),Translation=(X=0,Y=0,Z=0),Scal ...(truncated)"
    WeightMinimum=0
    Weight=&RigVMFunction_MathDoubleMul(
     B=&RigVMFunction_MathDoubleRemap(SourceMaximum=1, SourceMinimum=0, TargetMaximum=0, TargetMinimum=1, Value=0, bClamp=False).Result
     A=@n5.OutputParam
    ).Result
   )
   > &RigUnit_ModifyTransforms(
    ItemToModify=((Item=(Type=Bone,Name="base_scarf_r"),Transform=(Rotation=(X=-0.2028,Y=0.2896,Z=-0.096,W=0.9305),Translation=(X=2,Y=6,Z=4),Scale3D=(X=1,Y=1,Z=1))))
    WeightMinimum=0
    Weight=&RigVMFunction_MathDoubleMul(
     A=RigVMDispatchNode(BoneName=root, Default=0, Name=deformRig_r_upperarm_up_alpha) {template="DISPATCH_RigDispatch_GetAnimAttribute(in Name,in BoneName,in Default,out Value,out Found)"}.Value
     B=RigVMVariableNode() {variable=is_Parachuting}.Value
    ).Result
   )
   > n6 = &RigUnit_SphericalPoseReader(
    ActiveRegionScaleFactors=(PositiveWidth=1,NegativeWidth=1,PositiveHeight=1,NegativeHeight=1)
    ActiveRegionSize=0.07
    Debug=(bDrawDebug=false,bDraw2D=False,bDrawLocalAxes=False,DebugScale=25,DebugSegments=20,DebugThickness=0.25)
    DriverAxis=(X=-1,Y=0,Z=0)
    DriverItem=(Type=Bone,Name="clavicle_r")
    FalloffRegionScaleFactors=(PositiveWidth=1,NegativeWidth=1,PositiveHeight=1,NegativeHeight=1)
    FalloffSize=0.275
    FlipHeightScaling=False
    FlipWidthScaling=False
    OptionalParentItem=(Type=Bone,Name="spine_05")
    RotationOffset=(X=133,Y=-15,Z=90)
   )
   n7 = RigVMVariableNode() {variable=is_Parachuting}
   n8 = &RigVMFunction_AlphaInterp(
    InterpSpeedDecreasing=2.5
    InterpSpeedIncreasing=5
    Scale=1
    bClampResult=False
    Value=&RigVMFunction_MathDoubleMul(A=&RigVMFunction_MathDoubleMul(A=@n6.OutputParam, B=@n7.Value).Result, B=RigVMVariableNode() {variable=isMoving_right}.Value).Result
   )
   > &RigUnit_ModifyTransforms(
    ItemToModify=((Item=(Type=Bone,Name="base_scarf_r"),Transform=(Rotation=(X=0,Y=-0.1736,Z=0,W=0.9848),Translation=(X=-3,Y=-1,Z=8),Scale3D=(X=1,Y=1.5,Z=1))))
    WeightMinimum=0
    Weight=@n8.Result
   )
   n9 = &RigVMFunction_AlphaInterp(
    InterpSpeedDecreasing=5
    InterpSpeedIncreasing=5
    Scale=1
    bClampResult=False
    Value=&RigVMFunction_MathDoubleMul(B=RigVMVariableNode() {variable=isMoving_forward}.Value, A=RigVMVariableNode() {variable=is_Parachuting}.Value).Result
   )
   > &RigUnit_ModifyTransforms(
    ItemToModify=((Item=(Type=Bone,Name="base_scarf_r"),Transform=(Rotation=(X=0,Y=-0.1736,Z=0,W=0.9848),Translation=(X=-1,Y=0,Z=4),Scale3D=(X=1,Y=1.5,Z=1))))
    WeightMinimum=0
    Weight=@n9.Result
   )
   DeltaRotation v4 = RigVMFunctionReferenceNode(
    Rotation Order=ZYX
    child=(Type=Bone,Name="head")
    debugRotation=false
    outputMinMax=(X=0,Y=1)
    parent=(Type=Null,Name="head_null")
    premult=1
    rotationMinMax=(X=0,Y=-20)
    triggerAxis=Z
    twistAxis=(X=1,Y=0,Z=0)
    ExecuteContext=@RigVMFunction_Sequence.E
   )
   n10 = &RigVMFunction_AlphaInterp(InterpSpeedDecreasing=5, InterpSpeedIncreasing=5, Scale=1, bClampResult=False, Value=@"DeltaRotation v4".outputWeight)
   > &RigUnit_ModifyTransforms(
    ItemToModify=((Item=(Type=Bone,Name="base_scarf_back_middle"),Transform=(Rotation=(X=0,Y=0,Z=0,W=1),Translation=(X=4,Y=0,Z=-2),Scale3D=(X=1,Y=1,Z=1))))
    WeightMinimum=0
    Weight=@n10.Result
   )
   > DeltaRotation v4_1 = RigVMFunctionReferenceNode(
    Rotation Order=ZYX
    child=(Type=Bone,Name="foot_l")
    debugRotation=false
    outputMinMax=(X=0,Y=1)
    parent=(Type=Null,Name="foot_l_null")
    premult=1
    rotationMinMax=(X=0,Y=-45)
    triggerAxis=Z
    twistAxis=(X=1,Y=0,Z=0)
   )
   > &RigUnit_ModifyTransforms(
    ItemToModify=((Item=(Type=Bone,Name="dyn_foot_front_l"),Transform=(Rotation=(X=0,Y=0,Z=0,W=1),Translation=(X=5,Y=0,Z=-3),Scale3D=(X=1,Y=1,Z=1))))
    WeightMinimum=0
    Weight=@"DeltaRotation v4_1".outputWeight
   )
   > DeltaRotation v4_2 = RigVMFunctionReferenceNode(
    Rotation Order=ZYX
    child=(Type=Bone,Name="foot_r")
    debugRotation=false
    outputMinMax=(X=0,Y=1)
    parent=(Type=Null,Name="foot_r_null")
    premult=1
    rotationMinMax=(X=0,Y=-50)
    triggerAxis=Z
    twistAxis=(X=1,Y=0,Z=0)
   )
   > &RigUnit_ModifyTransforms(
    ItemToModify=((Item=(Type=Bone,Name="dyn_foot_front_r"),Transform=(Rotation=(X=0,Y=0.2588,Z=0,W=0.9659),Translation=(X=-5,Y=0,Z=6),Scale3D=(X=1,Y=1,Z=1))))
    WeightMinimum=0
    Weight=@"DeltaRotation v4_2".outputWeight
   )
   &RigUnit_ModifyTransforms(
    ItemToModify="((Item=(Type=Bone,Name=\"foot_base_l_dyn\"),Transform=(Rotation=(X=0,Y=0,Z=0,W=1),Translation=(X=-6,Y=-3,Z=0),Scale3D=(X=1,Y=1,Z=1))),(Item=(Type=Bone,Name=\"dyn_foot_front_l\"),Transform=(Rotation=(X=0,Y=-0.2588,Z=0,W=0.9659),Translation=(X=-5,Y=0,Z=0),Scale3 ...(truncated)"
    WeightMinimum=0
    Weight=RigVMVariableNode() {variable=isBallet_Dancing}.Value
    ExecutePin=@RigVMFunction_Sequence.F
   )
   > n11 = &RigUnit_SphericalPoseReader(
    ActiveRegionScaleFactors=(PositiveWidth=1,NegativeWidth=1,PositiveHeight=1,NegativeHeight=1)
    ActiveRegionSize=0.07
    Debug=(bDrawDebug=false,bDraw2D=False,bDrawLocalAxes=False,DebugScale=25,DebugSegments=20,DebugThickness=0.25)
    DriverAxis=(X=-1,Y=0,Z=0)
    DriverItem=(Type=Bone,Name="clavicle_r")
    FalloffRegionScaleFactors=(PositiveWidth=1,NegativeWidth=1,PositiveHeight=1,NegativeHeight=1)
    FalloffSize=0.275
    FlipHeightScaling=False
    FlipWidthScaling=False
    OptionalParentItem=(Type=Bone,Name="spine_05")
    RotationOffset=(X=133,Y=-15,Z=90)
   )
   Multiply = &RigVMFunction_MathDoubleMul(A=@n11.OutputParam, B=RigVMVariableNode() {variable=isBallet_Dancing}.Value)
   > &RigUnit_ModifyTransforms(
    ItemToModify=((Item=(Type=Bone,Name="dyn_scarf_r"),Transform=(Rotation=(X=0,Y=-0.4226,Z=0,W=0.9063),Translation=(X=3.5,Y=-3.5,Z=5.5),Scale3D=(X=1,Y=1,Z=1))))
    WeightMinimum=0
    Weight=@Multiply.Result
   )
   > DeltaRotation v4_3 = RigVMFunctionReferenceNode(
    Rotation Order=ZYX
    child=(Type=Bone,Name="foot_l")
    debugRotation=false
    outputMinMax=(X=0,Y=1)
    parent=(Type=Null,Name="foot_l_null")
    premult=1
    rotationMinMax=(X=0,Y=65)
    triggerAxis=Z
    twistAxis=(X=1,Y=0,Z=0)
   )
   > &RigUnit_ModifyTransforms(
    ItemToModify="((Item=(Type=Bone,Name=\"foot_base_l_dyn\"),Transform=(Rotation=(X=0,Y=0,Z=-0.1736,W=0.9848),Translation=(X=-7.5,Y=-2,Z=0),Scale3D=(X=1,Y=1,Z=1))),(Item=(Type=Bone,Name=\"dyn_foot_front_l\"),Transform=(Rotation=(X=0,Y=-0.1736,Z=0,W=0.9848),Translation=(X=-6,Y= ...(truncated)"
    WeightMinimum=0
    Weight=@"DeltaRotation v4_3".outputWeight
   )
   > DeltaRotation v4_4 = RigVMFunctionReferenceNode(
    Rotation Order=ZYX
    child=(Type=Bone,Name="foot_r")
    debugRotation=false
    outputMinMax=(X=0,Y=1)
    parent=(Type=Null,Name="foot_r_null")
    premult=1
    rotationMinMax=(X=0,Y=65)
    triggerAxis=Z
    twistAxis=(X=1,Y=0,Z=0)
   )
   > &RigUnit_ModifyTransforms(
    ItemToModify="((Item=(Type=Bone,Name=\"foot_base_r_dyn\"),Transform=(Rotation=(X=0,Y=-0.0872,Z=0,W=0.9962),Translation=(X=6,Y=1,Z=0),Scale3D=(X=1,Y=1,Z=1))),(Item=(Type=Bone,Name=\"dyn_foot_front_r\"),Transform=(Rotation=(X=0,Y=-0.1305,Z=0,W=0.9914),Translation=(X=3.5,Y=0,Z ...(truncated)"
    WeightMinimum=0
    Weight=@"DeltaRotation v4_4".outputWeight
   )
   &RigUnit_ModifyTransforms(
    ItemToModify=((Item=(Type=Bone,Name="dyn_scarf_r"),Transform=(Rotation=(X=0,Y=0,Z=0,W=-1),Translation=(X=1,Y=4,Z=2),Scale3D=(X=1,Y=1,Z=1))),(Item=(Type=Bone,Name="base_scarf_back"),Transform=(Rotation=(X=0,Y=0,Z=0,W=1),Translation=(X=0,Y=0,Z=3),Scale3D=(X=1,Y=1,Z=1))))
    WeightMinimum=0
    Weight=&RigVMFunction_MathDoubleMul(A=RigVMVariableNode() {variable=isSky Diving-Floating}.Value, B=RigVMVariableNode() {variable=isMoving_left}.Value).Result
    ExecutePin=@RigVMFunction_Sequence.A
   )
  ]}
 New Function_ContainedGraph {nodes [
   RigVMFunctionEntryNode()
   n1 = &RigVMFunction_AlphaInterp(
    InterpSpeedDecreasing=15
    InterpSpeedIncreasing=15
    Scale=1
    bClampResult=False
    Value=&RigVMFunction_MathBoolToFloat(Value=RigVMVariableNode() {variable=pub_isTargeting}.Value).Result
   )
   > RigVMVariableNode(Value=@n1.Result) {variable=isTargeting_Lerp}
   RigVMFunction_MathBoolOr = RigVMAggregateNode(
    A=RigVMVariableNode() {variable=pub_isAR_Equipped}.Value
    B=RigVMVariableNode() {variable=is_OutLaw_SG}.Value
    C=RigVMVariableNode() {variable=SMG_P90}.Value
   )
   n2 = &RigVMFunction_AlphaInterp(
    InterpSpeedDecreasing=15
    InterpSpeedIncreasing=15
    Scale=1
    bClampResult=False
    Value=&RigVMFunction_MathBoolToFloat(Value=@RigVMFunction_MathBoolOr.Result).Result
   )
   > RigVMVariableNode(Value=@n2.Result) {variable=isAR_Equipped_Lerp}
   n3 = &RigVMFunction_AlphaInterp(
    InterpSpeedDecreasing=15
    InterpSpeedIncreasing=15
    Scale=1
    bClampResult=False
    Value=&RigVMFunction_MathBoolToFloat(Value=RigVMVariableNode() {variable=pub_isPistol_Equipped}.Value).Result
   )
   > RigVMVariableNode(Value=@n3.Result) {variable=isPistol_Equipped_Lerp}
   > RigVMFunctionReturnNode()
  ]}
 RigVMFunctionLibrary {nodes [
   DeltaRotation v4 = RigVMCollapseNode(
    ExecuteContext=()
    Rotation Order=ZYX
    child=(Type=Bone,Name="None")
    outputMinMax=(X=0,Y=1)
    parent=(Type=Bone,Name="None")
    premult=1
    rotationMinMax=(X=0,Y=0)
    twistAxis=(X=1,Y=0,Z=0)
   )
   "Scarf_Adjustments " = RigVMCollapseNode(ExecuteContext=())
   Surface_Constraints = RigVMCollapseNode(ExecuteContext=())
   Variables = RigVMCollapseNode(ExecuteContext=())
  ]}
 RigVMFunction_MathBoolOr_1_ContainedGraph {nodes [
   n1 = RigVMFunctionEntryNode()
   RigVMFunction_MathBoolOr = &RigVMFunction_MathBoolOr(A=@n1.A, B=@n1.B)
   RigVMFunctionReturnNode(Result=&RigVMFunction_MathBoolOr(B=@n1.C, A=@RigVMFunction_MathBoolOr.Result).Result)
  ]}
 RigVMFunction_Sequence_1_ContainedGraph {nodes [
   n1 = RigVMFunctionEntryNode()
   RigVMFunction_Sequence = &RigVMFunction_Sequence(ExecuteContext=@n1.ExecuteContext)
   n2 = &RigVMFunction_Sequence(ExecuteContext=@RigVMFunction_Sequence.B)
   n3 = &RigVMFunction_Sequence(ExecuteContext=@n2.B)
   n4 = &RigVMFunction_Sequence(ExecuteContext=@n3.B)
   n5 = &RigVMFunction_Sequence(ExecuteContext=@n4.B)
   n6 = &RigVMFunction_Sequence(ExecuteContext=@n5.B)
   n7 = &RigVMFunction_Sequence(ExecuteContext=@n6.B)
   RigVMFunctionReturnNode(A=@RigVMFunction_Sequence.A, B=@n2.A, C=@n3.A, D=@n4.A, E=@n5.A, F=@n6.A, G=@n7.A, H=@n7.B)
  ]}
 RigVMModel {nodes [
   RigUnit_BeginExecution = &RigUnit_BeginExecution()
   FN_DeformRig_FNC = RigVMFunctionReferenceNode(
    DeformRig_ Curve="(deformRig_l_upperarm_up_alpha_72_#1=False,deformRig_r_upperarm_up_alpha_73_#2=true,deformRig_l_upperarm_fr_alpha_74_#3=False,deformRig_r_upperarm_fr_alpha_75_CB51B80 ...(truncated)"
    DeformRig_Bool="(upperarm_1_#4=True,elbow_3_#5=True,wrist_7_#6=True,groin_8_#7=True,glute_17_#8=True,knee_9_40C408324CA261 ...(truncated)"
    DeformRig_Debug=(upperarm_debug_18_#4=False,wrist_debug_20_#6=False,lat_debug_26_#9=False,groinGlute_debug_27_#7=False)
    DeformRig_Value="(upperarm_up_18_#10=(Rotation=(X=0.0,Y=-0.0,Z=0.0,W=1.0),Translation=(X=0.0,Y=0.0,Z=8.0),Scale3D=(X=0.70,Y=1.0,Z=1.0)),upperarm_fwd_fr_19_#11=(Rotation=(X=0.0,Y=-0.0,Z=0.0,W=1.0),Translation=(X=0.0, ...(truncated)"
    ExecuteContext=@RigUnit_BeginExecution.ExecutePin
   )
   > Surface_Constraints = RigVMFunctionReferenceNode()
   > Variables = RigVMFunctionReferenceNode()
   > "Scarf_Adjustments " = RigVMFunctionReferenceNode()
  ]}
]
hierarchy [
 root {loc=<0 0 0>, rot=<0 0 0 1>}
 pelvis {parent=root, loc=<0 -0.7071 93.1191>, rot=<0.0127 0.707 -0.0127 -0.707>}
 spine_01 {parent=pelvis, loc=<6.7494 -0.6388 0>, rot=<0 0 0.034 -0.9994>}
 spine_02 {parent=spine_01, loc=<7.2518 -0.035 0>, rot=<0 0 0.0238 -0.9997>}
 spine_03 {parent=spine_02, loc=<6.0104 0 0>, rot=<0 0 0.0149 -0.9999>}
 spine_04 {parent=spine_03, loc=<8.1024 0 0>, rot=<0 0 -0.0132 -0.9999>}
 spine_05 {parent=spine_04, loc=<10.0146 0 0>, rot=<0 0 -0.0015 -1>}
 neck_01 {parent=spine_05, loc=<11.3138 0.1529 0>, rot=<0 0 -0.2003 -0.9797>}
 neck_02 {parent=neck_01, loc=<4.5627 0 0>, rot=<0 0 0.0504 -0.9987>}
 head {parent=neck_02, loc=<5.3914 0 0>, rot=<0 0 0.115 -0.9934>}
 FACIAL_C_FacialRoot {parent=head, loc=<2.1524 -0.746 -0.0002>, rot=<0.4879 -0.5118 -0.4879 -0.5118>}
 FACIAL_R_NasolabialBulge {parent=FACIAL_C_FacialRoot, loc=<-3.3432 3.6232 6.8929>, rot=<0.0623 0.3774 -0.0257 -0.9236>}
 FACIAL_L_CheekLower {parent=FACIAL_C_FacialRoot, loc=<3.7069 3.4484 6.2725>, rot=<0.0389 -0.4265 0.0598 -0.9017>}
 FACIAL_R_CheekLower {parent=FACIAL_C_FacialRoot, loc=<-3.7036 3.4504 6.2677>, rot=<0.0389 0.4267 -0.0598 -0.9016>}
 FACIAL_L_Masseter {parent=FACIAL_C_FacialRoot, loc=<5.5318 1.2016 4.6225>, rot=<0.0741 0.1131 0.0073 -0.9908>}
 FACIAL_R_Masseter {parent=FACIAL_C_FacialRoot, loc=<-5.5324 1.2076 4.6158>, rot=<0.0741 -0.1128 -0.0074 -0.9908>}
 FACIAL_C_UnderChin {parent=FACIAL_C_FacialRoot, loc=<0.0009 8.1449 5.4293>, rot=<0.504 0.0002 0 -0.8637>}
 FACIAL_L_UnderChin {parent=FACIAL_C_FacialRoot, loc=<2.9607 6.9868 3.9466>, rot=<0.4066 -0.3543 0.084 -0.8379>}
 FACIAL_R_UnderChin {parent=FACIAL_C_FacialRoot, loc=<-2.9637 6.9914 3.9393>, rot=<0.4066 0.3546 -0.084 -0.8378>}
 FACIAL_C_Jaw {parent=FACIAL_C_FacialRoot, loc=<-0.0013 1.1843 2.9434>, rot=<0.3466 0.0002 0 -0.938>}
 FACIAL_C_Jawline {parent=FACIAL_C_Jaw, loc=<0.0011 2.5013 7.8999>, rot=<0.2586 0 0 -0.966>}
 FACIAL_L_ChinSide {parent=FACIAL_C_Jaw, loc=<2.553 0.698 7.3395>, rot=<-0.2208 -0.3738 0.0499 -0.8995>}
 FACIAL_R_ChinSide {parent=FACIAL_C_Jaw, loc=<-2.5468 0.7039 7.3373>, rot=<-0.2208 0.3738 -0.0499 -0.8995>}
 FACIAL_C_Chin {parent=FACIAL_C_Jaw, loc=<0.0006 1.1842 7.8634>, rot=<-0.2591 0 0 -0.9659>}
 FACIAL_L_Ear {parent=FACIAL_C_FacialRoot, loc=<6.4224 -0.5056 1.4997>, rot=<-0.0215 -0.0007 -0.0262 -0.9994>}
 FACIAL_R_Ear {parent=FACIAL_C_FacialRoot, loc=<-6.4264 -0.5071 1.4936>, rot=<-0.0215 0.001 0.0261 -0.9994>}
 FACIAL_L_Jawline {parent=FACIAL_C_FacialRoot, loc=<3.9019 5.1241 5.0234>, rot=<0.2139 0.1576 -0.0554 -0.9624>}
 FACIAL_R_Jawline {parent=FACIAL_C_FacialRoot, loc=<-3.898 5.123 5.0117>, rot=<0.2139 -0.1573 0.0554 -0.9625>}
 head_null {parent=neck_02, type=Null, loc=<5.3914 0 0>, rot=<0 0 0.115 -0.9934>}
 attach_cape {parent=spine_05, loc=<9.0221 -11.714 -0.0003>, rot=<0.0616 -0.7044 -0.0616 -0.7044>}
 cape {parent=attach_cape, type=Null, loc=<0 0.5734 -0.1378>, rot=<0.0821 0 0 0.9966>}
 attach_backpack {parent=spine_05, loc=<-0.1171 -13.4729 -0.0004>, rot=<-0.0093 -0.707 0.0093 -0.7071>}
 Backpack_BR {parent=attach_backpack, type=Null, loc=<0 2.6822 -0.0411>, rot=<-0.0429 0 0 0.9991>}
 clavicle_l {parent=spine_05, loc=<4.3575 -0.1466 -2.3223>, rot=<0.0465 -0.7289 -0.009 -0.6829>}
 upperarm_l {parent=clavicle_l, loc=<15.5111 0 0>, rot=<0.0226 -0.3234 -0.0088 -0.9459>}
 lowerarm_l {parent=upperarm_l, loc=<26.6514 0 0>, rot=<0.0391 -0.0099 -0.2453 -0.9686>}
 hand_l {parent=lowerarm_l, loc=<22.3491 0 0>, rot=<-0.7081 -0.0507 0.0975 -0.6975>}
 thumb_01_l {parent=hand_l, loc=<1.3118 -0.3191 -2.459>, rot=<0.6546 -0.4042 0.0664 -0.6354>}
 thumb_02_l {parent=thumb_01_l, loc=<5.9336 -0.0448 0.0401>, rot=<-0.0037 0.0213 0.298 -0.9543>}
 thumb_03_l {parent=thumb_02_l, loc=<2.2223 0.3682 -0.0907>, rot=<0.0118 -0.013 0.0066 -0.9998>}
 index_metacarpal_l {parent=hand_l, loc=<1.6535 0.0183 -2.96>, rot=<0.0158 -0.0124 -0.0173 -0.9996>}
 index_01_l {parent=index_metacarpal_l, loc=<8.3008 0.0357 -0.0524>, rot=<-0.0044 0.0485 0.0924 -0.9945>}
 index_02_l {parent=index_01_l, loc=<3.3587 -0.3004 0.0452>, rot=<-0.002 0.0027 0.3225 -0.9466>}
 index_03_l {parent=index_02_l, loc=<2.5578 -0.123 -0.0252>, rot=<-0.002 0.0067 0.2558 -0.9667>}
 middle_metacarpal_l {parent=hand_l, loc=<2.1856 -0.1725 -0.5984>, rot=<0.006 0.0168 -0.0329 -0.9993>}
 middle_01_l {parent=middle_metacarpal_l, loc=<7.3275 -0.2447 0.0129>, rot=<-0.0208 0.0042 0.0858 -0.9961>}
 middle_02_l {parent=middle_01_l, loc=<4.0123 0.1401 0.1175>, rot=<-0.0218 0.0258 0.3512 -0.9357>}
 middle_03_l {parent=middle_02_l, loc=<2.6077 0.0534 0.1373>, rot=<0.0082 0.0188 0.1708 -0.9851>}
 ring_metacarpal_l {parent=hand_l, loc=<2.2074 -0.3625 1.3638>, rot=<-0.0284 0.0722 -0.0458 -0.9959>}
 ring_01_l {parent=ring_metacarpal_l, loc=<7.0295 -0.6197 -0.1446>, rot=<-0.0489 -0.0252 0.1892 -0.9804>}
 ring_02_l {parent=ring_01_l, loc=<3.6174 0.4647 -0.1027>, rot=<0 0 0.2806 -0.9598>}
 ring_03_l {parent=ring_02_l, loc=<2.4152 0.114 0.0199>, rot=<0 0 0.1189 -0.9929>}
 pinky_metacarpal_l {parent=hand_l, loc=<1.7616 -0.9753 3.3814>, rot=<-0.0397 0.0814 -0.069 -0.9935>}
 pinky_01_l {parent=pinky_metacarpal_l, loc=<7.1709 -0.4831 -0.1094>, rot=<-0.0654 -0.0035 0.2242 -0.9723>}
 pinky_02_l {parent=pinky_01_l, loc=<2.8274 0.0893 -0.1043>, rot=<0.0037 0.0011 0.2854 -0.9584>}
 pinky_03_l {parent=pinky_02_l, loc=<2.2046 0.0307 -0.0078>, rot=<-0.0002 0.0098 0.1445 -0.9895>}
 deform_wrist_dn_base_l {parent=hand_l, loc=<0 -2.2342 0>, rot=<0 0 0 -1>}
 deform_wrist_dn_l {parent=deform_wrist_dn_base_l, loc=<0 0 0>, rot=<0 0 0 1>}
 deform_wrist_up_base_l {parent=hand_l, loc=<0 2.2682 0>, rot=<0 0 0 -1>}
 deform_wrist_up_l {parent=deform_wrist_up_base_l, loc=<0 0 0>, rot=<0 0 0 1>}
 lowerarm_twist_02_l {parent=lowerarm_l, loc=<9.6199 0 0>, rot=<0 0 0 -1>}
 lowerarm_twist_01_l {parent=lowerarm_l, loc=<17.8549 0 0>, rot=<0 0 0 -1>}
 deform_elbow_in_base_l {parent=lowerarm_l, loc=<0.2112 5.259 0.8274>, rot=<0 0 0 -1>}
 deform_elbow_in_l {parent=deform_elbow_in_base_l, loc=<0 0 0>, rot=<0 0 0 1>}
 deform_elbow_out_base_l {parent=lowerarm_l, loc=<0 -4 0>, rot=<0 0 0 -1>}
 deform_elbow_out_l {parent=deform_elbow_out_base_l, loc=<0 0 0>, rot=<0 0 0 1>}
 upperarm_twist_01_l {parent=upperarm_l, loc=<0.0041 0 0>, rot=<0 0 0 -1>}
 deform_upperarm_bk_base_l {parent=upperarm_twist_01_l, loc=<1.0813 -7.2193 -1.9453>, rot=<0 0 0 1>}
 deform_upperarm_bk_l {parent=deform_upperarm_bk_base_l, loc=<0 0 0>, rot=<0 0 0 1>}
 deform_upperarm_fr_base_l {parent=upperarm_twist_01_l, loc=<-0.3691 7.0506 -2.765>, rot=<0 0 0 1>}
 deform_upperarm_fr_l {parent=deform_upperarm_fr_base_l, loc=<0 0 0>, rot=<0 0 0 1>}
 deform_upperarm_up_base_l {parent=upperarm_twist_01_l, loc=<-3.2216 -1.1391 3.3311>, rot=<0 0 0 1>}
 deform_upperarm_up_l {parent=deform_upperarm_up_base_l, loc=<0 0 0>, rot=<0 0 0 1>}
 upperarm_twist_02_l {parent=upperarm_l, loc=<13.7941 0 0>, rot=<0 0 0 -1>}
 clavicle_r {parent=spine_05, loc=<4.3575 -0.1466 2.3223>, rot=<0.7289 0.0465 -0.6829 0.009>}
 upperarm_r {parent=clavicle_r, loc=<-15.5111 0 0>, rot=<0.0226 -0.3234 -0.0088 -0.9459>}
 lowerarm_r {parent=upperarm_r, loc=<-26.6514 0 0>, rot=<0.0391 -0.0099 -0.2453 -0.9686>}
 hand_r {parent=lowerarm_r, loc=<-22.3491 0 0>, rot=<-0.7081 -0.0507 0.0975 -0.6975>}
 thumb_01_r {parent=hand_r, loc=<-1.3118 0.3191 2.459>, rot=<0.6546 -0.4042 0.0664 -0.6354>}
 thumb_02_r {parent=thumb_01_r, loc=<-5.9336 0.0448 -0.0401>, rot=<-0.0037 0.0213 0.298 -0.9543>}
 thumb_03_r {parent=thumb_02_r, loc=<-2.2223 -0.3682 0.0907>, rot=<0.0118 -0.013 0.0066 -0.9998>}
 index_metacarpal_r {parent=hand_r, loc=<-1.6535 -0.0183 2.96>, rot=<0.0158 -0.0124 -0.0173 -0.9996>}
 index_01_r {parent=index_metacarpal_r, loc=<-8.3008 -0.0357 0.0524>, rot=<-0.0044 0.0485 0.0924 -0.9945>}
 index_02_r {parent=index_01_r, loc=<-3.3587 0.3004 -0.0452>, rot=<-0.002 0.0027 0.3225 -0.9466>}
 index_03_r {parent=index_02_r, loc=<-2.5578 0.123 0.0252>, rot=<-0.002 0.0067 0.2558 -0.9667>}
 middle_metacarpal_r {parent=hand_r, loc=<-2.1856 0.1725 0.5984>, rot=<0.006 0.0168 -0.0329 -0.9993>}
 middle_01_r {parent=middle_metacarpal_r, loc=<-7.3275 0.2447 -0.0129>, rot=<-0.0208 0.0042 0.0858 -0.9961>}
 middle_02_r {parent=middle_01_r, loc=<-4.0123 -0.1401 -0.1175>, rot=<-0.0218 0.0258 0.3512 -0.9357>}
 middle_03_r {parent=middle_02_r, loc=<-2.6077 -0.0534 -0.1373>, rot=<0.0082 0.0188 0.1708 -0.9851>}
 ring_metacarpal_r {parent=hand_r, loc=<-2.2074 0.3625 -1.3638>, rot=<-0.0284 0.0722 -0.0458 -0.9959>}
 ring_01_r {parent=ring_metacarpal_r, loc=<-7.0295 0.6197 0.1446>, rot=<-0.0489 -0.0252 0.1892 -0.9804>}
 ring_02_r {parent=ring_01_r, loc=<-3.6174 -0.4647 0.1027>, rot=<0 0 0.2806 -0.9598>}
 ring_03_r {parent=ring_02_r, loc=<-2.4152 -0.114 -0.0199>, rot=<0 0 0.1189 -0.9929>}
 pinky_metacarpal_r {parent=hand_r, loc=<-1.7616 0.9753 -3.3814>, rot=<-0.0397 0.0814 -0.069 -0.9935>}
 pinky_01_r {parent=pinky_metacarpal_r, loc=<-7.1709 0.4831 0.1094>, rot=<-0.0654 -0.0035 0.2242 -0.9723>}
 pinky_02_r {parent=pinky_01_r, loc=<-2.8274 -0.0893 0.1043>, rot=<0.0037 0.0011 0.2854 -0.9584>}
 pinky_03_r {parent=pinky_02_r, loc=<-2.2046 -0.0307 0.0078>, rot=<-0.0002 0.0098 0.1445 -0.9895>}
 deform_wrist_dn_base_r {parent=hand_r, loc=<-0.0004 2.2345 -0.0001>, rot=<0 0 0 -1>}
 deform_wrist_dn_r {parent=deform_wrist_dn_base_r, loc=<0 0 0>, rot=<0 0 0 1>}
 deform_wrist_up_base_r {parent=hand_r, loc=<0.0003 -2.2684 0.0001>, rot=<0 0 0 -1>}
 deform_wrist_up_r {parent=deform_wrist_up_base_r, loc=<0 0 0>, rot=<0 0 0 1>}
 lowerarm_twist_02_r {parent=lowerarm_r, loc=<-9.6199 0 0>, rot=<0 0 0 -1>}
 lowerarm_twist_01_r {parent=lowerarm_r, loc=<-17.8549 0 0>, rot=<0 0 0 -1>}
 deform_elbow_in_base_r {parent=lowerarm_r, loc=<-0.2111 -5.2591 -0.8275>, rot=<0 0 0 -1>}
 deform_elbow_in_r {parent=deform_elbow_in_base_r, loc=<-0.0001 0 0.0002>, rot=<0 0 0 1>}
 deform_elbow_in_r_Helper {parent=deform_elbow_in_r, loc=<0 0 0>, rot=<0 0 0 1>}
 deform_elbow_out_base_r {parent=lowerarm_r, loc=<0 4 -0.0001>, rot=<0 0 0 -1>}
 deform_elbow_out_r {parent=deform_elbow_out_base_r, loc=<0 0 0>, rot=<0 0 0 1>}
 upperarm_twist_01_r {parent=upperarm_r, loc=<-0.0041 0 0>, rot=<0 0 0 -1>}
 deform_upperarm_bk_base_r {parent=upperarm_twist_01_r, loc=<-1.0814 7.2192 1.9455>, rot=<0 0 0 1>}
 deform_upperarm_bk_r {parent=deform_upperarm_bk_base_r, loc=<0 0 0>, rot=<0 0 0 1>}
 deform_upperarm_fr_base_r {parent=upperarm_twist_01_r, loc=<0.3689 -7.0506 2.7652>, rot=<0 0 0 1>}
 deform_upperarm_fr_r {parent=deform_upperarm_fr_base_r, loc=<0 0 0>, rot=<0 0 0 1>}
 deform_upperarm_up_base_r {parent=upperarm_twist_01_r, loc=<3.2214 1.1391 -3.3308>, rot=<0 0 0 1>}
 deform_upperarm_up_r {parent=deform_upperarm_up_base_r, loc=<0 0 0>, rot=<0 0 0 -1>}
 upperarm_r_dup {parent=upperarm_twist_01_r, loc=<0.0041 0 0>, rot=<0 0 0 -1>}
 upperarm_twist_02_r {parent=upperarm_r, loc=<-13.7941 0 0>, rot=<0 0 0 -1>}
 base_scarf_center {parent=spine_05, loc=<11.533 2.7107 0.774>, rot=<-0.4812 -0.8734 -0.0414 -0.0626>}
 dyn_scarf_center {parent=base_scarf_center, loc=<0 0 0>, rot=<0 0 0 1>}
 base_scarf_back {parent=dyn_scarf_center, loc=<-8.7657 -4.916 -8.092>, rot=<0.5775 -0.4731 -0.496 0.4434>}
 dyn_scarf_back_01 {parent=base_scarf_back, loc=<0 0 0>, rot=<0 0 0 1>}
 dyn_scarf_back_02 {parent=dyn_scarf_back_01, loc=<9.5604 -0.1995 -0.6477>, rot=<0.0332 -0.1776 0.0241 -0.9832>}
 dyn_scarf_back_03 {parent=dyn_scarf_back_02, loc=<9.1364 -0.0045 -0.0112>, rot=<-0.0004 -0.0235 0.0056 -0.9997>}
 dyn_scarf_back_04 {parent=dyn_scarf_back_03, loc=<8.4277 0.0031 0.0038>, rot=<0.0182 0.0237 -0.0183 -0.9994>}
 base_scarf_front {parent=dyn_scarf_center, loc=<9.9924 -2.3558 -8.7779>, rot=<0.6113 -0.3103 0.2364 -0.6886>}
 dyn_scarf_front_01 {parent=base_scarf_front, loc=<0 0 0>, rot=<0 0.0064 0 -1>}
 dyn_scarf_front_02 {parent=dyn_scarf_front_01, loc=<7.7609 0.0146 0.496>, rot=<0.0285 -0.0906 -0.0058 -0.9955>}
 dyn_scarf_front_03 {parent=dyn_scarf_front_02, loc=<9.1086 0 0>, rot=<0.0169 -0.044 -0.0151 -0.9988>}
 dyn_scarf_front_04 {parent=dyn_scarf_front_03, loc=<8.2014 0 0>, rot=<0 0 0.0144 -0.9999>}
 base_scarf_back_middle {parent=dyn_scarf_center, loc=<-7.6289 1.9028 1.5749>, rot=<0.0024 0.6571 0.7525 -0.0442>}
 dyn_scarf_back_middle {parent=base_scarf_back_middle, loc=<0 0 0>, rot=<0 0 0 1>}
 base_scarf_front_middle {parent=dyn_scarf_center, loc=<8.4823 2.9882 0.5864>, rot=<0.6585 -0.0658 0.0114 -0.7496>}
 dyn_scarf_front_middle {parent=base_scarf_front_middle, loc=<0 0 0>, rot=<0 0 0 1>}
 base_scarf_r {parent=spine_05, loc=<12.9213 -2.4238 11.1083>, rot=<0.5517 0.0222 0.8331 -0.0334>}
 dyn_scarf_r {parent=base_scarf_r, loc=<0 0 0>, rot=<0 0 0 1>}
 deform_pec_aim_base_l {parent=spine_04, loc=<7.8732 12.6388 -2.4123>, rot=<0.0102 -0.4602 0.3979 -0.7936>}
 deform_pec_aim_l {parent=deform_pec_aim_base_l, loc=<0 0 0>, rot=<0 0 0 1>}
 deform_pec_base_l {parent=deform_pec_aim_l, loc=<10.766 0 0>, rot=<0 0 0 1>}
 deform_pec_l {parent=deform_pec_base_l, loc=<0.055 0 0>, rot=<0 0 0 1>}
 deform_pec_aim_base_r {parent=spine_04, loc=<7.8727 12.6388 2.4123>, rot=<-0.4602 -0.0102 0.7936 0.3979>}
 deform_pec_aim_r {parent=deform_pec_aim_base_r, loc=<0 0 0>, rot=<0 0 0 1>}
 deform_pec_base_r {parent=deform_pec_aim_r, loc=<-10.7662 -0.0006 -0.0007>, rot=<0 0 0 1>}
 deform_pec_r {parent=deform_pec_base_r, loc=<0 0 0>, rot=<0 0 0 1>}
 thigh_l {parent=pelvis, loc=<-2.5693 -0.0109 -7.7501>, rot=<0.0012 0.0572 0.0155 -0.9982>}
 calf_l {parent=thigh_l, loc=<-39.0484 0 0>, rot=<0.0006 -0.0161 -0.0391 -0.9991>}
 foot_l {parent=calf_l, loc=<-42.561 0.1077 0.4292>, rot=<0.0418 -0.0455 0.0343 -0.9975>}
 ball_l {parent=foot_l, loc=<-6.1599 13.8264 0.6806>, rot=<0.0001 -0.0001 -0.7063 -0.7079>}
 foot_base_l_helper {parent=foot_l, loc=<2.2618 -5.7525 0>, rot=<0 0 0 1>}
 foot_base_l_dyn {parent=foot_base_l_helper, loc=<0 0 0>, rot=<0 0 0 1>}
 base_foot_front_l {parent=foot_l, loc=<4.7726 6.2001 -0.2653>, rot=<-0.5504 -0.4467 -0.4922 -0.5052>}
 dyn_foot_front_l {parent=base_foot_front_l, loc=<0 0 0>, rot=<0 0 0 1>}
 calf_twist_02_l {parent=calf_l, loc=<-18.8064 -0.0967 0.0397>, rot=<0.0237 0.0095 -0.0035 -0.9997>}
 calf_twist_01_l {parent=calf_l, loc=<-35.813 -0.0967 0.0397>, rot=<0.0237 0.0095 -0.0035 -0.9997>}
 deform_knee_in_base_l {parent=calf_l, loc=<1.298 -5.2957 -0.0496>, rot=<0 0 0 -1>}
 deform_knee_in_l {parent=deform_knee_in_base_l, loc=<0 0 0>, rot=<0 0 0 1>}
 deform_knee_out_base_l {parent=calf_l, loc=<0 4 0>, rot=<0 0 0 -1>}
 deform_knee_out_l {parent=deform_knee_out_base_l, loc=<0 0 0>, rot=<0 0 0 1>}
 foot_l_null {parent=calf_l, type=Null, loc=<-42.561 0.1077 0.4292>, rot=<0.0418 -0.0455 0.0343 -0.9975>}
 base_outer_l_knee {parent=calf_l, loc=<-0.4226 -1.4314 -5.9383>, rot=<-0.0315 -0.7349 0.027 -0.6769>}
 dyn_outer_l_knee {parent=base_outer_l_knee, loc=<0 0 0>, rot=<0 0 0 1>}
 base_inner_l_knee {parent=calf_l, loc=<0.1891 -0.4448 6.6445>, rot=<0.7349 -0.0315 0.6769 0.027>}
 dyn_inner_l_knee {parent=base_inner_l_knee, loc=<0 0 0>, rot=<0 0 0 1>}
 thigh_twist_01_l {parent=thigh_l, loc=<-21.6303 0 0>, rot=<0.0013 0 0 -1>}
 deform_groin_base_l {parent=thigh_l, loc=<-2.1908 8.1142 -1.7442>, rot=<0 0 0 -1>}
 deform_groin_l {parent=deform_groin_base_l, loc=<0 0 0>, rot=<0 0 0 1>}
 groin_l_helper {parent=thigh_l, loc=<-1.8986 8.1814 -1.0391>, rot=<-0.4516 -0.5661 -0.4352 -0.535>}
 dyn_l_helper {parent=groin_l_helper, loc=<0 0 0>, rot=<0 0 0 1>}
 thigh_r {parent=pelvis, loc=<-2.5693 -0.0109 7.7501>, rot=<0.0572 -0.0012 0.9982 0.0155>}
 calf_r {parent=thigh_r, loc=<39.0484 0 0>, rot=<0.0006 -0.0161 -0.0391 -0.9991>}
 foot_r {parent=calf_r, loc=<42.561 -0.1077 -0.4292>, rot=<0.0418 -0.0455 0.0343 -0.9975>}
 ball_r {parent=foot_r, loc=<6.1599 -13.8264 -0.6806>, rot=<0.0001 -0.0001 -0.7063 -0.7079>}
 foot_base_r_helper {parent=foot_r, loc=<-2.2619 5.7525 0>, rot=<0 0 0 -1>}
 foot_base_r_dyn {parent=foot_base_r_helper, loc=<0 0 0>, rot=<0 0 0 -1>}
 base_foot_front_r {parent=foot_r, loc=<-4.7726 -6.2002 0.2653>, rot=<-0.5504 -0.4467 -0.4922 -0.5052>}
 dyn_foot_front_r {parent=base_foot_front_r, loc=<0 0 0>, rot=<0 0 0 1>}
 calf_twist_02_r {parent=calf_r, loc=<18.8064 0.0967 -0.0397>, rot=<0.0237 0.0095 -0.0035 -0.9997>}
 calf_twist_01_r {parent=calf_r, loc=<35.813 0.0967 -0.0397>, rot=<0.0237 0.0095 -0.0035 -0.9997>}
 deform_knee_in_base_r {parent=calf_r, loc=<-1.298 5.2957 0.0496>, rot=<0 0 0 -1>}
 deform_knee_in_r {parent=deform_knee_in_base_r, loc=<0 0 0>, rot=<0 0 0 1>}
 deform_knee_out_base_r {parent=calf_r, loc=<0 -4 0>, rot=<0 0 0 -1>}
 deform_knee_out_r {parent=deform_knee_out_base_r, loc=<0 0 0>, rot=<0 0 0 1>}
 foot_r_null {parent=calf_r, type=Null, loc=<42.561 -0.1077 -0.4292>, rot=<0.0418 -0.0455 0.0343 -0.9975>}
 base_outer_r_knee {parent=calf_r, loc=<0.8762 0.5873 5.8227>, rot=<-0.0315 -0.7349 0.027 -0.6769>}
 dyn_outer_r_knee {parent=base_outer_r_knee, loc=<0 0 0>, rot=<0 0 0 1>}
 base_inner_r_knee {parent=calf_r, loc=<-0.3349 0.3456 -6.2345>, rot=<0.7349 -0.0315 0.6769 0.027>}
 dyn_inner_r_knee {parent=base_inner_r_knee, loc=<0 0 0>, rot=<0 0 0 1>}
 thigh_twist_01_r {parent=thigh_r, loc=<21.6303 0 0>, rot=<0.0013 0 0 -1>}
 deform_groin_base_r {parent=thigh_r, loc=<2.1908 -8.1142 1.7442>, rot=<0 0 0 -1>}
 deform_groin_r {parent=deform_groin_base_r, loc=<0 0 0>, rot=<0 0 0 1>}
 deform_glute_base_l {parent=pelvis, loc=<-8.8064 -8.8222 -8.4287>, rot=<0.0275 0.0501 -0.4472 -0.8926>}
 deform_glute_l {parent=deform_glute_base_l, loc=<0 0 0>, rot=<0 0 0 1>}
 deform_glute_base_r {parent=pelvis, loc=<-8.8062 -8.8222 8.4287>, rot=<0.0501 -0.0275 0.8926 -0.4472>}
 deform_glute_r {parent=deform_glute_base_r, loc=<0 0 0>, rot=<0 0 0 1>}
 ik_hand_root {parent=root, loc=<0 0 0>, rot=<0 0 0 1>}
 ik_hand_gun {parent=ik_hand_root, loc=<-51.9727 4.0883 102.6935>, rot=<0.6084 0.1793 0.4613 -0.6205>}
]
variables [
 SMG_P90 {type=bool, default=False}
 isAR_Equipped_Lerp {type=double, default=0}
 isBallet_Dancing {type=double, default=0}
 isDBNO {type=bool, default=False}
 isKeytar_Emote {type=double, default=0}
 isMoving_forward {type=double, default=0}
 isMoving_left {type=double, default=0}
 isMoving_right {type=double, default=0}
 isPistol_Equipped_Lerp {type=double, default=0}
 isSky Diving-Floating {type=double, default=0}
 isTargeting_Lerp {type=double, default=0}
 is_OutLaw_SG {type=bool, default=False}
 is_Parachuting {type=double, default=0}
 pub_isAR_Equipped {type=bool, default=False}
 pub_isPistol_Equipped {type=bool, default=False}
 pub_isRigidBodies_Enabled {type=bool, default=False}
 pub_isTargeting {type=bool, default=False}
 pub_isWearingBackBling {type=bool, default=False}
]
