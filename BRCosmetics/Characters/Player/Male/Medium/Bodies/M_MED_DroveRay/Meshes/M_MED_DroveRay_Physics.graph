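# Physics asset for the DroveRay medium male player body cosmetic, apparently built for secondary motion on its scarf.
# Capsules cover the pelvis, spine, clavicles, upper arms, right forearm and right thigh, likely as scarf colliders, joined mostly by free-motion constraints.
# A four-link front scarf chain hangs from a base under spine_05 with tight swing and twist limits.
# The single right scarf link has zero-degree limits to its base.
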PhysicsAsset /BRCosmetics/Characters/Player/Male/Medium/Bodies/M_MED_DroveRay/Meshes/M_MED_DroveRay_Physics
bodies {
 base_scarf_front [{shapeType=Capsule, length=2, radius=2, shapeName=base_scarf_front_capsule}]
 base_scarf_r [{shapeType=Capsule, length=0.505, radius=0.505, shapeName=base_scarf_r_capsule, rotation=<rot 0 -90 0>}]
 clavicle_l [{shapeType=Capsule, length=10, radius=7.5, shapeName=clavicle_l_capsule, center=<4 0 0>, rotation=<rot 90 0 0>}]
 clavicle_r [{shapeType=Capsule, length=10, radius=7, shapeName=clavicle_r_capsule, center=<-4 0 0>, rotation=<rot 90 180 0>}]
 dyn_scarf_front_01 [{shapeType=Capsule, length=10, radius=2.35, shapeName=dyn_scarf_front_01_capsule, center=<3 -1.5414 2.3252>, rotation=<rot 0.0002 0 100.407>}]
 dyn_scarf_front_02 [{
   shapeType=Capsule
   length=10
   radius=2.4
   shapeName=dyn_scarf_front_02_capsule
   center=<3.5 -1.28 1.5>
   rotation=<rot 0.0013 -0.0001 101.9717>
  }]
 dyn_scarf_front_03 [{
   shapeType=Capsule
   length=10
   radius=2.4
   shapeName=dyn_scarf_front_03_capsule
   center=<3 -1.4877 0.7245>
   rotation=<rot 0.0007 -0.0001 100.4207>
  }]
 dyn_scarf_front_04 [{
   shapeType=Capsule
   length=10
   radius=2.5
   shapeName=dyn_scarf_front_04_capsule
   center=<3 -1.1593 0.7939>
   rotation=<rot 0.0222 -0.0017 101.0033>
  }]
 dyn_scarf_r [{shapeType=Capsule, length=4.5, radius=1.75, shapeName=dyn_scarf_r_capsule, rotation=<rot 76.4485 89.9999 -180>}]
 lowerarm_r [{shapeType=Capsule, length=15, radius=4.8, shapeName=lowerarm_r_capsule, center=<-8.0097 -0.1091 0>, rotation=<rot 90 -5.1023 0>}]
 pelvis [{
   shapeType=Capsule
   length=7.3331
   radius=11.5673
   shapeName=pelvis_capsule
   center=<-1.1476 -1.0896 0.0585>
   rotation=<rot 0.3144 -90 0.1309>
  }]
 spine_01 [{
   shapeType=Capsule
   length=7.8368
   radius=11.7134
   shapeName=spine_01_capsule
   center=<1.9464 -0.0878 -0.1472>
   rotation=<rot 1.9672 -90 -0.3479>
  }]
 spine_02 [{shapeType=Capsule, length=8, radius=10, center=<1 -0.3273 -0.2356>}]
 spine_03 [{shapeType=Capsule, length=8, radius=10.2, shapeName=spine_03_capsule, center=<1 0.3232 0>}]
 spine_04 [{shapeType=Capsule, length=8, radius=11.75, center=<1.0421 -0.0671 -0.5324>}]
 spine_05 [{shapeType=Capsule, length=10, radius=10.75, center=<0.4321 -0.2327 0>}]
 thigh_r [{shapeType=Capsule, length=28.5, radius=8, shapeName=thigh_r_capsule, center=<19.9849 -0.1074 -0.4589>, rotation=<rot 88.2738 3.3899 0.0006>}]
 upperarm_l [{shapeType=Capsule, length=18, radius=6, shapeName=upperarm_l_capsule, center=<6.6728 -0.9342 -1.1423>, rotation=<rot 88.676 -0.0057 -4.7847>}]
 upperarm_r [{
   shapeType=Capsule
   length=26
   radius=5.1
   shapeName=upperarm_r_capsule
   center=<-7.059 1.6406 1.1437>
   rotation=<rot 89.7125 -152.5889 22.6334>
  }]
}
constraints [
 {bone1Name=spine_04, bone2Name=spine_03, swing1LimitDegrees=45, swing2LimitDegrees=45, twistLimitDegrees=45}
 {bone1Name=spine_05, bone2Name=spine_04, swing1LimitDegrees=45, swing2LimitDegrees=45, twistLimitDegrees=45}
 {
  bone1Name=spine_03
  bone2Name=spine_02
  swing1LimitDegrees=45
  swing1Motion=Free
  swing2LimitDegrees=45
  swing2Motion=Free
  twistLimitDegrees=45
  twistMotion=Free
 }
 {
  bone1Name=base_scarf_front
  bone2Name=spine_05
  swing1LimitDegrees=45
  swing1Motion=Free
  swing2LimitDegrees=45
  swing2Motion=Free
  twistLimitDegrees=45
  twistMotion=Free
 }
 {bone1Name=dyn_scarf_front_01, bone2Name=base_scarf_front, swing1LimitDegrees=25, swing2LimitDegrees=30, twistLimitDegrees=7.5}
 {bone1Name=dyn_scarf_front_02, bone2Name=dyn_scarf_front_01, swing1LimitDegrees=25, swing2LimitDegrees=30, twistLimitDegrees=5}
 {bone1Name=dyn_scarf_front_03, bone2Name=dyn_scarf_front_02, swing1LimitDegrees=20, swing2LimitDegrees=30, twistLimitDegrees=5}
 {bone1Name=dyn_scarf_front_04, bone2Name=dyn_scarf_front_03, swing1LimitDegrees=25, swing2LimitDegrees=30, twistLimitDegrees=5}
 {
  bone1Name=clavicle_l
  bone2Name=spine_05
  swing1LimitDegrees=45
  swing1Motion=Free
  swing2LimitDegrees=45
  swing2Motion=Free
  twistLimitDegrees=45
  twistMotion=Free
 }
 {
  bone1Name=upperarm_l
  bone2Name=clavicle_l
  swing1LimitDegrees=45
  swing1Motion=Free
  swing2LimitDegrees=45
  swing2Motion=Free
  twistLimitDegrees=45
  twistMotion=Free
 }
 {
  bone1Name=clavicle_r
  bone2Name=spine_05
  swing1LimitDegrees=45
  swing1Motion=Free
  swing2LimitDegrees=45
  swing2Motion=Free
  twistLimitDegrees=45
  twistMotion=Free
 }
 {
  bone1Name=upperarm_r
  bone2Name=clavicle_r
  swing1LimitDegrees=45
  swing1Motion=Free
  swing2LimitDegrees=45
  swing2Motion=Free
  twistLimitDegrees=45
  twistMotion=Free
 }
 {
  bone1Name=base_scarf_r
  bone2Name=spine_05
  swing1LimitDegrees=45
  swing1Motion=Free
  swing2LimitDegrees=45
  swing2Motion=Free
  twistLimitDegrees=45
  twistMotion=Free
 }
 {bone1Name=dyn_scarf_r, bone2Name=base_scarf_r, swing1LimitDegrees=0, swing2LimitDegrees=0, twistLimitDegrees=0}
 {
  bone1Name=lowerarm_r
  bone2Name=upperarm_r
  swing1LimitDegrees=45
  swing1Motion=Free
  swing2LimitDegrees=45
  swing2Motion=Free
  twistLimitDegrees=45
  twistMotion=Free
 }
 {
  bone1Name=spine_02
  bone2Name=spine_01
  swing1LimitDegrees=45
  swing1Motion=Free
  swing2LimitDegrees=45
  swing2Motion=Free
  twistLimitDegrees=45
  twistMotion=Free
 }
 {
  bone1Name=spine_01
  bone2Name=pelvis
  swing1LimitDegrees=45
  swing1Motion=Free
  swing2LimitDegrees=45
  swing2Motion=Free
  twistLimitDegrees=45
  twistMotion=Free
 }
 {
  bone1Name=thigh_r
  bone2Name=pelvis
  swing1LimitDegrees=45
  swing1Motion=Free
  swing2LimitDegrees=45
  swing2Motion=Free
  twistLimitDegrees=45
  twistMotion=Free
 }
]
