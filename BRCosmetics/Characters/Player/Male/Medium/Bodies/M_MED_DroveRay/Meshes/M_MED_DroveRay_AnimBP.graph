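# Post-process AnimBlueprint for the DroveRay medium male body cosmetic, adding scarf and secondary-bone motion to the attached parent's copied pose.
# A Control Rig, fed movement, weapon, backbling and emote states, fills a cached pose.
# Two RigidBody branches, one chaining a rear physics asset, blend by bool before constraint and additive offsets adjust scarf, knee and groin bones.
# The event graph derives those states, scarf physics alpha and parameter-asset gravity overrides.
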AnimBlueprint /BRCosmetics/Characters/Player/Male/Medium/Bodies/M_MED_DroveRay/Meshes/M_MED_DroveRay_AnimBP
target_skeleton=&/M_MED_DroveRay_Skeleton
graphs [
 AnimGraph {nodes [
   Copy Pose From Mesh() {bUseAttachedParent=true}
   n1 = "Transform (Modify) Bone - Bone: attach_backpack"(
    ComponentPose=Local To Component(LocalPose=Use cached pose 'ControlRig_Cache'())
    Alpha=1
    Translation="0,-2.2,0"
    Rotation="0,0,1"
    Scale="1,1,1"
   ) {
    rotationMode=BMM_Additive
    rotationSpace=BCS_BoneSpace
    translationMode=BMM_Additive
    translationSpace=BCS_BoneSpace
    boneToModify {boneName=attach_backpack}
    rotation {roll=1}
    translation {y=-2.2}
   }
   n2 = RigidBody!(
    ComponentPose=@n1
    Alpha
    OverrideWorldGravity:Vector="0,0,0"
    ExternalForce
    SimSpaceSettings:Sim Space Settings Structure=(WorldAlpha=0.3,VelocityScaleZ=0.05,DampingAlpha=0.4,MaxLinearVelocity=100,MaxAngularVelocity=60,MaxLinearAcceleration=500,MaxAngularAcceleration=40,ExternalLinearDragV=(X=2,Y=2,Z=2),ExternalLinearVelocity=(X=0,Y=0,Z=0),ExternalAngularVelocity=(X=0,Y=0,Z=0))
   ) -> (Pose) {
    bOverrideWorldGravity=true
    overridePhysicsAsset=&/M_MED_DroveRay_Physics
    baseBoneRef {boneName=spine_05}
    simSpaceSettings {
     dampingAlpha=0.4
     maxAngularAcceleration=40
     maxAngularVelocity=60
     maxLinearAcceleration=500
     maxLinearVelocity=100
     velocityScaleZ=0.05
     worldAlpha=0.3
     externalLinearDragV=<2 2 2>
    }
   }
   n3 = Component To Local(ComponentPose=@n2)
   n4 = RigidBody(
    ComponentPose=Local To Component(LocalPose=Use cached pose 'ControlRig_Cache'())
    OverrideWorldGravity="0,0,0"
    SimSpaceSettings=(WorldAlpha=0.275,VelocityScaleZ=0.05,DampingAlpha=0.4,MaxLinearVelocity=100,MaxAngularVelocity=60,MaxLinearAcceleration=500,MaxAngularAcceleration=40,ExternalLinearDragV=(X=2,Y=2,Z=2),ExternalLinearVelocity=(X=0,Y=0,Z=0),ExternalAngularVelocity=(X=0,Y=0,Z=0))
   ) {
    bOverrideWorldGravity=true
    overridePhysicsAsset=&M_MED_DroveRay_Physics
    simulationSpace=BaseBoneSpace
    baseBoneRef {boneName=spine_05}
    simSpaceSettings {
     dampingAlpha=0.4
     maxAngularAcceleration=40
     maxAngularVelocity=60
     maxLinearAcceleration=500
     maxLinearVelocity=100
     velocityScaleZ=0.05
     worldAlpha=0.275
     externalLinearDragV=<2 2 2>
    }
   }
   n5 = RigidBody(
    ComponentPose=@n4
    OverrideWorldGravity="0,0,0"
    SimSpaceSettings=(WorldAlpha=0.275,VelocityScaleZ=0.05,DampingAlpha=0.4,MaxLinearVelocity=100,MaxAngularVelocity=60,MaxLinearAcceleration=500,MaxAngularAcceleration=40,ExternalLinearDragV=(X=2,Y=2,Z=2),ExternalLinearVelocity=(X=0,Y=0,Z=0),ExternalAngularVelocity=(X=0,Y=0,Z=0))
   ) {
    bOverrideWorldGravity=true
    overridePhysicsAsset=&/M_MED_DroveRayRear_Physics
    simulationSpace=BaseBoneSpace
    alphaScaleBiasClamp {bInterpResult=true, interpSpeedDecreasing=5, interpSpeedIncreasing=5}
    baseBoneRef {boneName=spine_05}
    simSpaceSettings {
     dampingAlpha=0.4
     maxAngularAcceleration=40
     maxAngularVelocity=60
     maxLinearAcceleration=500
     maxLinearVelocity=100
     velocityScaleZ=0.05
     worldAlpha=0.275
     externalLinearDragV=<2 2 2>
    }
   }
   n6 = Component To Local(ComponentPose=@n5)
   n7 = "Constraint - dyn_scarf_back_03 (P)"!(
    ComponentPose=Local To Component(LocalPose=Blend Poses by bool(BlendPose_0=@n3, BlendPose_1=@n6))
    bAlphaBoolEnabled:bool=True
    ConstraintWeights_0=1
   ) -> (Pose) {alphaInputType=Bool, boneToModify {boneName=dyn_scarf_back_03}, constraintSetup [{targetBone {boneName=spine_03}}], constraintWeights [1]}
   n8 = "Transform (Modify) Bone - Bone: dyn_scarf_back_03"!(ComponentPose=@n7, bAlphaBoolEnabled:bool=True, Translation="0,0,-3", Rotation="0,0,0", Scale="1,1,1") -> (Pose) {
    alphaInputType=Bool
    translationMode=BMM_Additive
    translationSpace=BCS_BoneSpace
    alphaBoolBlend {blendInTime=1, blendOutTime=1}
    boneToModify {boneName=dyn_scarf_back_03}
    translation {z=-3}
   }
   n9 = "Transform (Modify) Bone - Bone: dyn_outer_l_knee"(ComponentPose=@n8, Alpha=1, Translation="-2.8,-3,-1", Rotation="0,0,0", Scale="1,1,1") {
    rotationMode=BMM_Additive
    translationMode=BMM_Additive
    alphaScaleBiasClamp {bInterpResult=true, interpSpeedDecreasing=5, interpSpeedIncreasing=5}
    boneToModify {boneName=dyn_outer_l_knee}
    translation=<-2.8 -3 -1>
   }
   n10 = "Transform (Modify) Bone - Bone: dyn_outer_r_knee"(ComponentPose=@n9, Alpha=1, Translation="3.621,-2.000,1.148", Rotation="0,0,0", Scale="1,1,1") {
    rotationMode=BMM_Additive
    translationMode=BMM_Additive
    alphaScaleBiasClamp {bInterpResult=true, interpSpeedDecreasing=5, interpSpeedIncreasing=5}
    boneToModify {boneName=dyn_outer_r_knee}
    translation=<3.621 -2 1.148>
   }
   n11 = "Transform (Modify) Bone - Bone: dyn_inner_r_knee"(ComponentPose=@n10, Alpha=1, Translation="-2.973,-11.649,-1.048", Rotation="-84.676,180.000,-180.000", Scale="1,1,1") {
    rotationMode=BMM_Additive
    translationMode=BMM_Additive
    alphaScaleBiasClamp {bInterpResult=true, interpSpeedDecreasing=5, interpSpeedIncreasing=5}
    boneToModify {boneName=dyn_inner_r_knee}
    rotation=<rot -84.676 180 -180>
    translation=<-2.973 -11.649 -1.048>
   }
   n12 = "Transform (Modify) Bone - Bone: dyn_inner_l_knee"(ComponentPose=@n11, Alpha=1, Translation="2.5,-3,-4", Rotation="0,0,0", Scale="1,1,1") {
    rotationMode=BMM_Additive
    translationMode=BMM_Additive
    alphaScaleBiasClamp {bInterpResult=true, interpSpeedDecreasing=5, interpSpeedIncreasing=5}
    boneToModify {boneName=dyn_inner_l_knee}
    translation=<2.5 -3 -4>
   }
   n13 = "Transform (Modify) Bone - Bone: base_scarf_back"!(ComponentPose=@n12, bAlphaBoolEnabled:bool=True, Translation="0,0,4.5", Rotation="0,0,6", Scale="1,1,1") -> (Pose) {
    alphaInputType=Bool
    rotationMode=BMM_Additive
    rotationSpace=BCS_BoneSpace
    translationMode=BMM_Additive
    translationSpace=BCS_BoneSpace
    alphaBoolBlend {blendInTime=0.2, blendOutTime=0.2}
    boneToModify {boneName=base_scarf_back}
    rotation {roll=6}
    translation {z=4.5}
   }
   n14 = "Transform (Modify) Bone - Bone: base_scarf_r"!(ComponentPose=@n13, bAlphaBoolEnabled:bool=True, Translation="0,2,4", Rotation="0,0,0", Scale="1,1,1") -> (Pose) {
    alphaInputType=Bool
    translationMode=BMM_Additive
    translationSpace=BCS_BoneSpace
    alphaBoolBlend {blendInTime=0.2, blendOutTime=0.2}
    boneToModify {boneName=base_scarf_r}
    translation {y=2, z=4}
   }
   n15 = "Transform (Modify) Bone - Bone: base_scarf_back"(ComponentPose=@n14, bAlphaBoolEnabled=True, Translation="0,0,3.5", Rotation="0,0,3", Scale="1,1,1") {
    alphaInputType=Bool
    rotationMode=BMM_Additive
    rotationSpace=BCS_BoneSpace
    translationMode=BMM_Additive
    translationSpace=BCS_BoneSpace
    alphaBoolBlend {blendInTime=0.2, blendOutTime=0.2}
    alphaScaleBiasClamp {bInterpResult=true, interpSpeedIncreasing=5}
    boneToModify {boneName=base_scarf_back}
    rotation {roll=3}
    translation {z=3.5}
   }
   n16 = "Transform (Modify) Bone - Bone: groin_l_helper"!(ComponentPose=@n15, bAlphaBoolEnabled:bool=True, Translation="-11,-2,0", Rotation="50,0,0", Scale="1,1,1") -> (Pose) {
    alphaInputType=Bool
    rotationMode=BMM_Additive
    translationMode=BMM_Additive
    boneToModify {boneName=groin_l_helper}
    rotation {pitch=50}
    translation=<-11 -2>
   }
   Output Pose(Result=Component To Local(ComponentPose=@n16)) {layerGroup=None}
   n17 = Control Rig!(
    Source=Copy Pose From Mesh() {bUseAttachedParent=true}
    Alpha
    pub_isRigidBodies_Enabled:bool=Get bRigidBodyEnabled()
    pub_isWearingBackBling:bool=Get isWearingBlackbling() -> (isWearingBlackbling:bool)
    pub_isTargeting:bool=Get bIsTargeting()
    pub_isAR_Equipped:bool=Get isAR Equipped() -> (isAR Equipped:bool)
    pub_isPistol_Equipped:bool=Get isPistol Equipped() -> (isPistol Equipped:bool)
    is_OutLaw_SG:bool=Get Is_outLaw_SG() -> (Is_outLaw_SG:bool)
    SMG_P90:bool=Get SMG_SG() -> (SMG_SG:bool)
    isDBNO:bool=Get bIsDBNO()
    is_Parachuting:double=Get Is_Parachute_Open() -> (Is_Parachute_Open:double)
    isBallet_Dancing:double=Get IsBallet_or_Keytar() -> (IsBallet_or_Keytar:double)
    isMoving_right:double=Get strafeRight float() -> (strafeRight float:double)
    isMoving_forward:double=Get moveForward() -> (moveForward:double)
    isKeytar_Emote:double=Get isKeytar_emoting() -> (isKeytar_emoting:double)
    isSky Diving-Floating:double=Get is_skyDiving_Floating() -> (is_skyDiving_Floating:double)
    isMoving_left:double=Get strafeLeft float() -> (strafeLeft float:double)
   ) -> (Pose) {controlRigAssetReference {blueprintRigClass=&/M_MED_DroveRay_CtrlRig.M_MED_DroveRay_CtrlRig_C}}
   Save cached pose 'ControlRig_Cache'(Pose=@n17) {cacheName=ControlRig_Cache}
  ]}
 EventGraph {nodes [
   Event BlueprintUpdateAnimation()
   > n1 = Sequence!(execute) -> (then_0, then_1, then_2:Exec, then_3:Exec)
   > n2 = Sequence!(execute) -> (then_0, then_1, then_2, then_3, then_4:Exec)
   Set Is_Parachute_Open(execute=@n2.then_1, Is_Parachute_Open:double="To Float (Boolean)"(InBool=Get bIsParachuteOpened())) -> (Output_Get:double)
   n3 = Branch(execute=@n2.then_4, Condition=Get bIsPlayingMeleeAnim())
   > Set alpha ScarfRigidBody(alpha ScarfRigidBody:double=0.9) -> (Output_Get:double)
   n4 = Branch(execute=@n3.else, Condition=Get isTacticalJumpingFalling() -> (isTacticalJumpingFalling:bool))
   > Set alpha ScarfRigidBody(alpha ScarfRigidBody=0.8)
   n5 = Branch(execute=@n4.else, Condition=Get bIsOnGrindRail())
   > Set alpha ScarfRigidBody(alpha ScarfRigidBody=0.7)
   n6 = Branch(execute=@n5.else, Condition=Get bIsCrouching())
   > Set alpha ScarfRigidBody(alpha ScarfRigidBody=0.85)
   n7 = Branch(execute=@n6.else, Condition=Get bIsJumping())
   > Set alpha ScarfRigidBody(alpha ScarfRigidBody=1)
   Set alpha ScarfRigidBody(execute=@n7.else, alpha ScarfRigidBody=1)
   n8 = NOT Boolean(A=Get bIsMoving2D())
   n9 = NOT Boolean(A=Get bIsParachuteOpened())
   n10 = Get bIsSkydiveFloating()
   n11 = Get bIsSkyDiving()
   n12 = Branch(execute=@n2.then_2, Condition=AND Boolean(A=OR Boolean(A=@n10, B=@n11), B=OR Boolean(A=@n9, B=@n8)))
   > Set is_skyDiving_Floating(is_skyDiving_Floating:double=1) -> (Output_Get:double)
   Set is_skyDiving_Floating(execute=@n12.else, is_skyDiving_Floating=0)
   n13 = Branch(execute=@n2.then_0, Condition=OR Boolean!(A=Get bIsCrouching(), B=Get bIsCrouchSprinting(), C:bool) -> (ReturnValue))
   > Set Crou_Or_CroSpring(Crou_Or_CroSpring:double=1) -> (Output_Get:double)
   Set Crou_Or_CroSpring(execute=@n13.else, Crou_Or_CroSpring=0)
   n14 = OR Boolean!(A=Get bIsSkydiveFloating(), B=Get bIsSkyDiving()) -> (ReturnValue)
   n15 = Get bIsJumping()
   n16 = Get bIsFalling()
   n17 = Branch(execute=@n2.then_3, Condition=AND Boolean(A=OR Boolean(A=@n15, B=@n16), B=NOT Boolean(A=@n14)))
   > Set Is Jumping(Is Jumping:double=1) -> (Output_Get:double)
   Set Is Jumping(execute=@n17.else, Is Jumping=0)
   n18 = "Gravity Override State Function (Parameter Asset) (Nativized)"(
    execute=@n1.then_1
    CharacterPartAnimInstance=Self-Reference()
    GameTimeInSeconds=GetGameTimeInSeconds()
    DebugJointName=spine_05
    Parameters=&/F_MED_droveRay_Parameters
   )
   > Set Gravity Override(Gravity Override=@n18)
   > Set RBAN Settings Out(RBAN Settings Out=@n18.OutRigidBodyAnimNodeParameters)
   n19 = GetAnimInstance(self=GetOwnerSkeletalMeshComponent())
   Is Valid(exec=@n1.then_2, InputObject=@n19)
   > n20 = Set Is_Ballet_Dancing(Is_Ballet_Dancing:bool=OR Boolean(
     A=Montage_IsPlaying(self=@n19, Montage=/BRCosmetics/Animation/Game/MainPlayer/Emotes/BalletSpin/Emote_BalletSpin_CMM_M)
     B=Montage_IsPlaying(self=@n19, Montage=/BRCosmetics/Animation/Game/MainPlayer/Emotes/Takeout/Emote_Takeout_CMM_M)
    )) -> (Output_Get:bool)
   > n21 = Branch(Condition=@n20)
   > Set IsBallet_or_Keytar(IsBallet_or_Keytar:double=1) -> (Output_Get:double)
   Set IsBallet_or_Keytar(execute=@n21.else, IsBallet_or_Keytar=0)
   n22 = PlayerMovement(execute=@n1.then_3, AnimInstance=Self-Reference(), XDotProductCoefficient=1, YDotProductCoefficient=1, ZDotProductCoefficient=1)
   n23 = BreakVector(InVec=MakeVector(X=@n22, Y=@n22.YMovement))
   > Set strafeLeft float(strafeLeft float:double="Absolute (Float)"(A="Clamp (Float)"(Value=@n23, Min=-1, Max=0))) -> (Output_Get:double)
   > Set strafeRight float(strafeRight float:double="Clamp (Float)"(Value=@n23, Min=0, Max=1)) -> (Output_Get:double)
   > Set moveForward(moveForward:double="Clamp (Float)"(Value=@n23.Y, Min=0, Max=1)) -> (Output_Get:double)
   > Set moveBackward(moveBackward:double="Absolute (Float)"(A="Clamp (Float)"(Value=@n23.Y, Min=-1, Max=0))) -> (Output_Get:double)
   Event OnEndCrouching()
   > Set EndCrouching(EndCrouching:bool=true) -> (Output_Get:bool)
   > Delay(Duration=0.2)
   > Set EndCrouching(EndCrouching=false)
   Event OnCharacterCustomizationLoaded()
   n24 = CheckForBackbling()
   > Set isWearingBlackbling(isWearingBlackbling:bool=OR Boolean!(A=@n24, B=@n24.bIsBackblingACape, C:bool=@n24.bIsBackblingLong) -> (ReturnValue)) -> (Output_Get:bool)
   Event OnWeaponChanged()
   > Set isAR Equipped(isAR Equipped:bool=false) -> (Output_Get:bool)
   > Set isPistol Equipped(isPistol Equipped:bool=false) -> (Output_Get:bool)
   > Set Is_outLaw_SG(Is_outLaw_SG:bool=false) -> (Output_Get:bool)
   > Set SMG_SG(SMG_SG:bool=false) -> (Output_Get:bool)
   > n25 = Switch on EFortWeaponCoreAnimation!(execute, Selection:EFortWeaponCoreAnimation Enum=Get WeaponCoreAnim!() -> (WeaponCoreAnim)) -> (
    Melee:Exec
    Pistol:Exec
    Shotgun:Exec
    PaperBlueprint:Exec
    Rifle:Exec
    MeleeOneHand:Exec
    MachinePistol:Exec
    RocketLauncher:Exec
    GrenadeLauncher:Exec
    GoingCommando:Exec
    AssaultRifle:Exec
    TacticalShotgun:Exec
    SniperRifle:Exec
    TrapPlacement:Exec
    ShoulderLauncher:Exec
    AbilityDecoTool:Exec
    Crossbow:Exec
    C4:Exec
    RemoteControl:Exec
    DualWield:Exec
    AR_BullPup:Exec
    AR_ForwardGrip:Exec
    MedPackPaddles:Exec
    SMG_P90:Exec
    AR_DrumGun:Exec
    Consumable_Small:Exec
    Consumable_Large:Exec
    Balloon:Exec
    MountedTurret:Exec
    CreativeTool:Exec
    ExplosiveBow:Exec
    AshtonIndigo:Exec
    AshtonChicago:Exec
    MeleeDualWield:Exec
    Unarmed:Exec
   )
   Set isAR Equipped(execute=@n25.AssaultRifle, isAR Equipped=true)
   Set isPistol Equipped(execute=@n25.Pistol, isPistol Equipped=true)
   Set SMG_SG(execute=@n25.SMG_P90, SMG_SG=true)
   Set Is_outLaw_SG(execute=@n25.TacticalShotgun, Is_outLaw_SG=true)
  ]}
]
variables [
 alpha ScarfRigidBody {type=number}
 Crou_Or_CroSpring {type=number}
 EndCrouching {type=boolean}
 Gravity Override {type=Vector}
 Is Jumping {type=number}
 Is_Ballet_Dancing {type=boolean}
 Is_outLaw_SG {type=boolean}
 Is_Parachute_Open {type=number}
 is_skyDiving_Floating {type=number}
 isAR Equipped {type=boolean}
 IsBallet_or_Keytar {type=number}
 isKeytar_Emote {type=boolean}
 isKeytar_emoting {type=number}
 isPistol Equipped {type=boolean}
 isTacticalJumpingFalling {type=boolean}
 isWearingBlackbling {type=boolean}
 moveBackward {type=number}
 moveForward {type=number}
 RBAN Settings Out {type=FortRigidBodyAnimNodeParameters}
 SMG_SG {type=boolean}
 strafeLeft float {type=number}
 strafeRight float {type=number}
]
